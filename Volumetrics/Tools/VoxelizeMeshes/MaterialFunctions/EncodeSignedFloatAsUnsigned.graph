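# Material function that packs a signed Float3 in the -2 to 2 range into unsigned values, in the mesh voxelization tools.
# The input is split into components, each run through an identical Custom HLSL node, then recombined into a Float3.
# Each node clamps to just inside the range, zeroes denormal-sized values and multiplies negatives by -32768, storing the sign in the top exponent bit of a 6e5 float.
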MaterialFunction /Volumetrics/Tools/VoxelizeMeshes/MaterialFunctions/EncodeSignedFloatAsUnsigned
graph {nodes [
  MaterialExpressionMaterialFunctionCall_16 = MaterialFunctionCall(Float3=FunctionInput() {bUsePreviewValueAsDefault=true, description=Should be in the -2 to 2 range., inputName=Signed Float}) {materialFunction=&BreakOutFloat3Components}
  n1 = Custom!(Velocity=@MaterialExpressionMaterialFunctionCall_16.R) {
   code="\t// Only handles values in range (-2,2)\r\n\tVelocity = clamp( Velocity, -1.9999, 1.9999 );\r\n\t// Kill denormals\r\n\tVelocity = abs( Velocity ) < exp2(-14.0) ? 0 : Velocity;\r\n\t// Store sign in top bit of 6e5 exponent\r\n\tfloat EncodedVelocity = Velocity * ( Velocity < 0 ? -32768.0 : 1 );\r\n\r\nreturn EncodedVelocity;"
   description=EncodeNegativeVelocity
   outputType=CMOT_Float1
   inputs [{
     inputName=Velocity
     input {expression=&self:MaterialExpressionMaterialFunctionCall_16}
    }]
  }
  n2 = Custom(Velocity=@MaterialExpressionMaterialFunctionCall_16.G) {
   code="\t// Only handles values in range (-2,2)\r\n\tVelocity = clamp( Velocity, -1.9999, 1.9999 );\r\n\t// Kill denormals\r\n\tVelocity = abs( Velocity ) < exp2(-14.0) ? 0 : Velocity;\r\n\t// Store sign in top bit of 6e5 exponent\r\n\tfloat EncodedVelocity = Velocity * ( Velocity < 0 ? -32768.0 : 1 );\r\n\r\nreturn EncodedVelocity;"
   description=EncodeNegativeVelocity
   outputType=CMOT_Float1
   inputs [{
     inputName=Velocity
     input {
      expression=&self:MaterialExpressionMaterialFunctionCall_16
      outputIndex=1
     }
    }]
  }
  n3 = Custom(Velocity=@MaterialExpressionMaterialFunctionCall_16.B) {
   code="\t// Only handles values in range (-2,2)\r\n\tVelocity = clamp( Velocity, -1.9999, 1.9999 );\r\n\t// Kill denormals\r\n\tVelocity = abs( Velocity ) < exp2(-14.0) ? 0 : Velocity;\r\n\t// Store sign in top bit of 6e5 exponent\r\n\tfloat EncodedVelocity = Velocity * ( Velocity < 0 ? -32768.0 : 1 );\r\n\r\nreturn EncodedVelocity;"
   description=EncodeNegativeVelocity
   outputType=CMOT_Float1
   inputs [{
     inputName=Velocity
     input {
      expression=&self:MaterialExpressionMaterialFunctionCall_16
      outputIndex=2
     }
    }]
  }
  FunctionOutput(None=MaterialFunctionCall(X=@n1, Y=@n2, Z=@n3) {materialFunction=&MakeFloat3}.Result)
 ]}
