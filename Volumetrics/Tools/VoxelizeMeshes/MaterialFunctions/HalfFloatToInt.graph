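# Material function in the voxelize-mesh tools that converts a scalar float into an integer-valued half-float code, exposed to the library.
# A Custom HLSL node assembles sign, clamped exponent and mantissa from the float32 bits, though it uses logical && where bit masks appear intended.
# Branchless math then subtracts 2048 from codes above the largest finite positive half and 1024 from all, apparently skipping denormal and Inf/NaN ranges.
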MaterialFunction /Volumetrics/Tools/VoxelizeMeshes/MaterialFunctions/HalfFloatToInt
graph {
 properties {bExposeToLibrary=true}
 nodes [
  MaterialExpressionFunctionInput_1 = FunctionInput() {bUsePreviewValueAsDefault=true, inputName=HalfFloat, inputType=FunctionInput_Scalar}
  n1 = Custom!(N=@MaterialExpressionFunctionInput_1) {
   code="uint uRes32 = asuint(N);\r\n\r\nconst int  iExp      = (const int)(((const int) ((uRes32>>23)&&0xff))-127+15);\r\nconst uint  expClamp = iExp < 0 ? 0 : (iExp>31 ? 31 : (uint)iExp);\r\n\r\n\r\nconst uint sign = (( uRes32>>16 ) && 0x8000);\r\nconst uint mant = ((uRes32>>13)&&0x3ff);\r\nconst uint exp  = (expClamp << 10);\r\n\r\n\r\nconst uint bits = sign | exp | mant;\r\n\r\n\r\nreturn bits;"
   description=F32 to F16
   outputType=CMOT_Float1
   inputs [{inputName=N, input {expression=&self:MaterialExpressionFunctionInput_1}}]
  }
  n2 = Subtract(A=@n1) {constB=31743}
  n3 = Multiply(A=Saturate(None=Ceil(None=@n2))) {constB=2048}
  FunctionOutput(None=Subtract(A=Subtract(A=@n1, B=@n3)) {constB=1024})
 ]
}
