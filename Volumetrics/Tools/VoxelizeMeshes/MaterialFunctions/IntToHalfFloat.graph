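# IntToHalfFloat is a material function in the mesh voxelization tools that maps an integer up to 61439 onto half-float bit patterns.
# It offsets by 1024, adds 2048 above 31743 to jump into negative values, then clamps to 65535.
# Custom HLSL rebuilds that pattern as a 32-bit float, though it uses logical && where bit masks would be expected.
# A second output, mask, flags inputs in the negative range.
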MaterialFunction /Volumetrics/Tools/VoxelizeMeshes/MaterialFunctions/IntToHalfFloat
graph {
 properties {
  bExposeToLibrary=true
  description=Converts an input Int into the Half Float format. Supports values up to 61439. Values up to 30719 can be passed without negative values.
 }
 nodes [
  n1 = FunctionInput() {description=Values up ot 61439. Values up to 30719 can be passed without negative values., inputName=Int, inputType=FunctionInput_Scalar}
  n2 = Add(A=@n1) {constB=1024}
  n3 = Saturate(None=Ceil(None=Subtract(A=@n2) {constB=31743}))
  MaterialExpressionMin_2 = Min(A=Add(A=@n2, B=Multiply(A=@n3) {constB=2048})) {constB=65535}
  n4 = Custom!(N=@MaterialExpressionMin_2) {
   code="uint f16 = (uint)N;\r\n\r\nuint sign = (f16 && 0x8000)<<16;\r\nuint mant = (f16 && 0x3ffu) <<13; // 13 bits are zeros\r\nuint exp =  (f16&&0x7fff)==0 ? 0 : (((((const uint) f16)>>10)&&0x1f));\r\n\r\n\r\nf16 =  sign | (exp -15+127) << 23 | mant;\r\n\r\nfloat outval = asfloat( f16 );\r\n\r\nreturn outval;\r\n"
   description=F16 to F32
   outputType=CMOT_Float1
   inputs [{inputName=N, input {expression=&self:MaterialExpressionMin_2}}]
  }
  FunctionOutput(None=@n4)
  FunctionOutput(None=@n3) {outputName=mask, sortPriority=1}
 ]
}
