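# Material function for the voxelize-meshes tool that unpacks motion vectors as current minus previous world position, scaled by VolumeWorldSize.
# A selectable UV channel (0-2) is offset into an XYFrames atlas cell by idx and snapped to texel centers.
# Only the current position gets the dilate step: nested lerps of center and four neighbors, each weighted by the sample itself.
# Both positions decode via HalfFloatToInt and a constant divide before subtraction.
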MaterialFunction /Volumetrics/Tools/VoxelizeMeshes/MaterialFunctions/UnPackMotionVectorsWithDilate
graph {nodes [
  n1 = Ceil(None=ScalarParameter() {parameterName=UVChannel})
  n2 = LinearInterpolate(
   A=LinearInterpolate(A=TextureCoordinate(), B=TextureCoordinate() {coordinateIndex=1}, Alpha=Saturate(None=@n1))
   B=TextureCoordinate() {coordinateIndex=2}
   Alpha=Saturate(None=Subtract(A=@n1))
  )
  n3 = ScalarParameter() {defaultValue=1, parameterName=XYFrames}
  n4 = ScalarParameter() {defaultValue=1, parameterName=XYFrames}
  n5 = MaterialFunctionCall!(1D Index=ScalarParameter() {parameterName=idx}, Cells XY=@n4) {materialFunction=/Engine/Functions/Engine_MaterialFunctions01/Coordinates/1Dto2DIndex}
  n6 = ScalarParameter() {defaultValue=256, parameterName=SceneCaptureResolution}
  n7 = Multiply(A=Add(A=Divide(A=@n2, B=@n3), B=Divide(A=@n5.2D Index, B=@n4)), B=@n6) {constB=256}
  n8 = Add(A=Divide(A=Floor(None=@n7), B=@n6) {constB=256}, B=Divide(B=Multiply(A=@n6) {constB=2}) {constB=512})
  n9 = TextureSampleParameter2D!(UVs=@n8, Level, Apply View MipBias) {
   constMipValue=0
   mipValueMode=TMVM_MipLevel
   parameterName=CurrentWorldPosition
   samplerType=SAMPLERTYPE_LinearColor
   texture=/Volumetrics/Content/VolumeTextures/Textures/RT_Default
  }
  n10 = MaterialFunctionCall!(OffsetAmount=Divide(B=ScalarParameter() {defaultValue=256, parameterName=SceneCaptureResolution})) {materialFunction=/Engine/Functions/Engine_MaterialFunctions01/ImageAdjustment/BlurSampleOffsets}
  n11 = TextureSampleParameter2D(UVs=Add(A=@n8, B=@n10."-1,0")) {
   constMipValue=0
   mipValueMode=TMVM_MipLevel
   parameterName=CurrentWorldPosition
   samplerType=SAMPLERTYPE_LinearColor
   texture=&RT_Default
  }
  n12 = TextureSampleParameter2D(UVs=Add(A=@n8, B=@n10."0,-1")) {
   constMipValue=0
   mipValueMode=TMVM_MipLevel
   parameterName=CurrentWorldPosition
   samplerType=SAMPLERTYPE_LinearColor
   texture=&RT_Default
  }
  n13 = TextureSampleParameter2D(UVs=Add(A=@n8, B=@n10."1,0")) {
   constMipValue=0
   mipValueMode=TMVM_MipLevel
   parameterName=CurrentWorldPosition
   samplerType=SAMPLERTYPE_LinearColor
   texture=&RT_Default
  }
  n14 = TextureSampleParameter2D(UVs=Add(A=@n8, B=@n10."0,1")) {
   constMipValue=0
   mipValueMode=TMVM_MipLevel
   parameterName=CurrentWorldPosition
   samplerType=SAMPLERTYPE_LinearColor
   texture=&RT_Default
  }
  n15 = LinearInterpolate(A=@n12.RGB, B=LinearInterpolate(A=@n13.RGB, B=@n14.RGB, Alpha=@n13.RGB), Alpha=@n12.RGB)
  n16 = MaterialFunctionCall(Float3=LinearInterpolate(A=@n9.RGB, B=LinearInterpolate(A=@n11.RGB, B=@n15, Alpha=@n11.RGB), Alpha=@n9.RGB)) {materialFunction=&BreakOutFloat3Components}
  n17 = Divide(A=MaterialFunctionCall(
    X=MaterialFunctionCall!(HalfFloat=@n16.R) {materialFunction=&/HalfFloatToInt}.Result
    Y=MaterialFunctionCall(HalfFloat=@n16.G) {materialFunction=&HalfFloatToInt}.Result
    Z=MaterialFunctionCall(HalfFloat=@n16.B) {materialFunction=&HalfFloatToInt}.Result
   ) {materialFunction=&MakeFloat3}.Result) {constB=30719}
  n18 = TextureSampleParameter2D(UVs=@n8) {
   constMipValue=0
   mipValueMode=TMVM_MipLevel
   parameterName=PreviousWorldPosition
   samplerType=SAMPLERTYPE_LinearColor
   texture=&RT_Default
  }
  n19 = MaterialFunctionCall(Float3=@n18.RGB) {materialFunction=&BreakOutFloat3Components}
  n20 = Divide(A=MaterialFunctionCall(
    X=MaterialFunctionCall(HalfFloat=@n19.R) {materialFunction=&HalfFloatToInt}.Result
    Y=MaterialFunctionCall(HalfFloat=@n19.G) {materialFunction=&HalfFloatToInt}.Result
    Z=MaterialFunctionCall(HalfFloat=@n19.B) {materialFunction=&HalfFloatToInt}.Result
   ) {materialFunction=&MakeFloat3}.Result) {constB=30719}
  FunctionOutput(None=Multiply(A=Subtract(A=@n17, B=@n20), B=ScalarParameter() {defaultValue=256, parameterName=VolumeWorldSize}))
 ]}
